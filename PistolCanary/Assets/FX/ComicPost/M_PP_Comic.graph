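# Post-process material that gives the scene a comic-book look, applied to scene color after depth of field.
# Cel shading quantizes brightness in log2 steps while keeping color ratios, with a depth-based skybox mask.
# Posterization builds halftone dots on an aspect-corrected grid, sized by cell luminance.
# Outlines come from four-tap depth and world-normal edge detection with distance fade, alongside a textured comic frame and master blend controls.
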Material /PistolCanary/Assets/FX/ComicPost/M_PP_Comic
graph {
 properties {
  bAutomaticallySetUsageInEditor=false
  blendableLocation=BL_SceneColorAfterDOF
  blendablePriority=1
  materialDomain=MD_PostProcess
  pixelDepthOffsetMode=PDOM_Legacy
  usedShadingModels=Default Lit
 }
 nodes [
  n1 = Multiply(A=Distance(A=NamedRerouteUsage(), B=NamedRerouteUsage())) {constB=1.4142, desc=Normalise}
  n2 = Divide(A=NamedRerouteUsage(), B=NamedRerouteUsage())
  n3 = NamedRerouteUsage()
  n4 = ComponentMask(None=SceneTexture(UVs=Divide(A=@n2, B=@n3)) {sceneTextureId=PPI_PostProcessInput0}.Color) {g=true, r=true}
  n5 = OneMinus(None=Saturate(None=Desaturation(None=@n4)))
  n6 = Saturate(None=Divide(A=Subtract(A=@n1, B=@n5)) {constB=0.01})
  n7 = LinearInterpolate(
   A=SceneTexture() {sceneTextureId=PPI_PostProcessInput0}.Color
   B=@n6
   Alpha=ScalarParameter() {defaultValue=0.1, group=2. Posterize, parameterName=Posterize Amount, sliderMax=1}
  )
  LinearInterpolate(
   A=ComponentMask(None=SceneTexture() {sceneTextureId=PPI_PostProcessInput0}.Color) {a=true, b=true, g=true, r=true}
   B=@n7
   Alpha=ScalarParameter() {group=0. VFX Adjustments, parameterName=MasterVFXAdjustment, sortPriority=1}
  )
  n8 = TextureSample() {texture=/Game/Creative/PostProcess/Textures/T_ComicFrame}
  n9 = LinearInterpolate(A=LinearInterpolate(B=SceneTexture() {sceneTextureId=PPI_PostProcessInput0}.Color, Alpha=@n8.R) {constA=1}, Alpha=@n8.G) {constB=0}
  LinearInterpolate(
   A=ComponentMask(None=SceneTexture() {sceneTextureId=PPI_PostProcessInput0}.Color) {a=true, b=true, g=true, r=true}
   B=@n9
   Alpha=ScalarParameter() {defaultValue=1, group=Main, parameterName=Master FX Control, sortPriority=0}
  )
  n10 = ComponentMask(None=SceneTexture() {sceneTextureId=PPI_PostProcessInput0}.Color) {b=true, g=true, r=true}
  n11 = Max(A=ComponentMask(None=@n10) {r=true}, B=Max(A=ComponentMask(None=@n10) {g=true}, B=ComponentMask(None=@n10) {b=true}))
  n12 = ScalarParameter() {defaultValue=5, group=1. Cell Shading, parameterName=Steps}
  n13 = Round(None=Multiply(A=Logarithm2!(X=@n11), B=@n12))
  n14 = Multiply(A=Divide(A=@n10, B=@n11), B=Power(Base=Constant() {r=2}, Exp=Divide(A=@n13, B=@n12)))
  n15 = ComponentMask(None=@n14) {b=true, g=true, r=true}
  LinearInterpolate(
   A=ComponentMask(None=SceneTexture() {sceneTextureId=PPI_PostProcessInput0}.Color) {b=true, g=true, r=true}
   B=@n15
   Alpha=ScalarParameter() {defaultValue=1, group=Main, parameterName=Master FX Control, sortPriority=0}
  )
  n16 = SceneTexture() {sceneTextureId=PPI_PostProcessInput0}
  LinearInterpolate(
   A=LinearInterpolate(A=@n14, B=ComponentMask(None=@n16.Color) {b=true, g=true, r=true}, Alpha=NamedRerouteUsage())
   B=@n14
   Alpha=ScalarParameter() {group=1. Cell Shading, parameterName=Cel Shade Sky}
  )
  n17 = NamedRerouteUsage()
  n18 = NamedRerouteUsage()
  LinearInterpolate(
   A=ComponentMask(None=SceneTexture() {sceneTextureId=PPI_PostProcessInput0}.Color) {b=true, g=true, r=true}
   B=StaticSwitchParameter(True=Min(A=@n17, B=@n18), False=@n17) {parameterName=Add Posterization}
   Alpha=ScalarParameter() {defaultValue=1, group=Main, parameterName=Master FX Control, sortPriority=0}
  )
  Min(A=@n6)
  n19 = MaterialFunctionCall!(
   Input=SceneDepth().R
   Input Low=ScalarParameter() {group=3. Outlines, parameterName=Near Range}
   Input High=ScalarParameter() {defaultValue=30000, group=3. Outlines, parameterName=Far Range}
  ) {materialFunction=/Engine/Functions/Engine_MaterialFunctions03/Math/RemapValueRangeNormalized}
  n20 = VectorParameter() {group=3. Outlines, parameterName=Outline Colour, defaultValue=<rgba 0 0 0 1>}
  n21 = Constant()
  n22 = NamedRerouteUsage()
  n23 = Max(
   A=StaticSwitchParameter(True=NamedRerouteUsage(), False=@n21) {defaultValue=true, group=3. Outlines, parameterName=WN Lines}
   B=StaticSwitchParameter(True=@n22, False=@n21) {defaultValue=true, group=3. Outlines, parameterName=SD Lines}
  )
  n24 = LinearInterpolate(
   B=LinearInterpolate(Alpha=FeatureLevelSwitch(Default=@n23, ES2=@n22, ES3_1=@n22, SM4_REMOVED=@n22)) {desc="Pre TM Adjust (if needed)"}
   Alpha=@n20.A
  )
  n25 = LinearInterpolate(A=ComponentMask(None=SceneTexture() {sceneTextureId=PPI_PostProcessInput0}.Color) {b=true, g=true, r=true}, B=@n20.RGB, Alpha=@n24)
  n26 = LinearInterpolate(A=@n25, B=ComponentMask(None=SceneTexture() {sceneTextureId=PPI_PostProcessInput0}.Color) {b=true, g=true, r=true}, Alpha=@n19.Result)
  n27 = ScalarParameter() {defaultValue=1, group=3. Outlines, parameterName=Outline Strength}
  Multiply(A=@n26, B=@n27)
  n28 = ViewSize()
  n29 = ComponentMask(None=@n28) {r=true}
  n30 = ComponentMask(None=@n28) {g=true}
  NamedRerouteDeclaration(None=AppendVector(A=Divide(A=@n29, B=@n30), B=Constant() {r=1})) {name=Aspect Ratio, nodeColor=<rgba 0 0.7656 1 1>}
  n31 = TextureCoordinate()
  n32 = NamedRerouteUsage()
  n33 = NamedRerouteDeclaration(None=Multiply(A=Multiply(A=@n31, B=@n32), B=NamedRerouteUsage())) {name=Scaled UVs, nodeColor=<rgba 0 1 0.4609 1>}
  NamedRerouteDeclaration(None=Add(A=Floor(None=@n33)) {constB=0.5}) {name=CellCoord, nodeColor=<rgba 0.6875 1 0 1>}
  n34 = SceneTexture() {sceneTextureId=PPI_PostProcessInput0}
  NamedRerouteDeclaration(None=LinearInterpolate(A=ComponentMask(None=@n34.Color) {b=true, g=true, r=true}, B=ComponentMask(None=@n26) {b=true, g=true, r=true}, Alpha=@n27)) {name=Outline Emissive Out, nodeColor=<rgba 0.6875 0 1 1>}
  NamedRerouteDeclaration(None=@n8.R) {name=Comic Border Mask, nodeColor=<rgba 1 0.0234 0 1>}
  NamedRerouteDeclaration(None=ScalarParameter() {defaultValue=140, desc=Grid Res, group=2. Posterize, parameterName=Grid Res}) {name=Grid Res, nodeColor=<rgba 0.8281 0 1 1>}
  n35 = SceneDepth()
  n36 = Multiply(A=SceneTexture() {sceneTextureId=PPI_SceneDepth}.InvSize, B=Constant2Vector() {g=1})
  n37 = Constant()
  n38 = LinearInterpolate(
   A=ScalarParameter() {defaultValue=2, group=3. Outlines, parameterName=SD Far Thickness}
   B=ScalarParameter() {defaultValue=1, group=3. Outlines, parameterName=SD Near Thickness}
   Alpha=If(
    A=SceneTexture() {sceneTextureId=PPI_SceneDepth}.Color
    B=ScalarParameter() {defaultValue=10000, group=3. Outlines, parameterName=Line thickness distance}
    "A > B"=@n37
    "A == B"=@n37
    "A < B"=Constant() {r=1}
   )
  ) {constA=1, constB=10}
  n39 = SceneDepth(UVs=Add(A=TextureCoordinate(), B=Multiply(A=@n36, B=@n38)))
  n40 = SceneDepth()
  n41 = Multiply(A=SceneTexture() {sceneTextureId=PPI_SceneDepth}.InvSize, B=Constant2Vector() {g=-1})
  n42 = SceneDepth(UVs=Add(A=TextureCoordinate(), B=Multiply(A=@n41, B=@n38)))
  n43 = SceneDepth()
  n44 = Multiply(A=SceneTexture() {sceneTextureId=PPI_SceneDepth}.InvSize, B=Constant2Vector() {r=1})
  n45 = SceneDepth(UVs=Add(A=TextureCoordinate(), B=Multiply(A=@n44, B=@n38)))
  n46 = SceneDepth()
  n47 = Multiply(A=SceneTexture() {sceneTextureId=PPI_SceneDepth}.InvSize, B=Constant2Vector() {r=-1})
  n48 = SceneDepth(UVs=Add(A=TextureCoordinate(), B=Multiply(A=@n47, B=@n38)))
  n49 = Add(A=Add(A=Subtract(A=@n35.R, B=@n39.R), B=Subtract(A=@n40.R, B=@n42.R)), B=Add(A=Subtract(A=@n43.R, B=@n45.R), B=Subtract(A=@n46.R, B=@n48.R)))
  n50 = ScalarParameter() {defaultValue=10, group=3. Outlines, parameterName=SD Threshold}
  NamedRerouteDeclaration(None=Saturate(None=Divide(A=@n49, B=@n50))) {name=SD Lines, nodeColor=<rgba 0.7812 0 1 1>}
  n51 = TextureCoordinate()
  n52 = ScalarParameter() {defaultValue=0.1, group=3. Outlines, parameterName=WN Thickness}
  n53 = Multiply(A=Multiply(A=SceneTexture() {sceneTextureId=PPI_SceneDepth}.InvSize, B=Constant2Vector() {g=1}), B=@n52)
  n54 = Subtract(A=SceneTexture() {sceneTextureId=PPI_WorldNormal}.Color, B=SceneTexture(UVs=Add(A=@n51, B=@n53)) {sceneTextureId=PPI_WorldNormal}.Color)
  n55 = TextureCoordinate()
  n56 = Multiply(A=Multiply(A=SceneTexture() {sceneTextureId=PPI_SceneDepth}.InvSize, B=Constant2Vector() {g=-1}), B=@n52)
  n57 = Subtract(A=SceneTexture() {sceneTextureId=PPI_WorldNormal}.Color, B=SceneTexture(UVs=Add(A=@n55, B=@n56)) {sceneTextureId=PPI_WorldNormal}.Color)
  n58 = TextureCoordinate()
  n59 = Multiply(A=Multiply(A=SceneTexture() {sceneTextureId=PPI_SceneDepth}.InvSize, B=Constant2Vector() {r=1}), B=@n52)
  n60 = Subtract(A=SceneTexture() {sceneTextureId=PPI_WorldNormal}.Color, B=SceneTexture(UVs=Add(A=@n58, B=@n59)) {sceneTextureId=PPI_WorldNormal}.Color)
  n61 = TextureCoordinate()
  n62 = Multiply(A=Multiply(A=SceneTexture() {sceneTextureId=PPI_SceneDepth}.InvSize, B=Constant2Vector() {r=-1}), B=@n52)
  n63 = Subtract(A=SceneTexture() {sceneTextureId=PPI_WorldNormal}.Color, B=SceneTexture(UVs=Add(A=@n61, B=@n62)) {sceneTextureId=PPI_WorldNormal}.Color)
  n64 = Desaturation(None=Add(A=Add(A=@n54, B=@n57), B=Add(A=@n60, B=@n63)))
  n65 = ScalarParameter() {defaultValue=0.05, group=3. Outlines, parameterName=WN Threshold}
  NamedRerouteDeclaration(None=Saturate(None=Power(Base=@n64, Exp=@n65) {constExponent=1, desc=Line Detail})) {name=WN Lines, nodeColor=<rgba 0 1 0.4609 1>}
  NamedRerouteDeclaration(None=ComponentMask(None=@n7) {b=true, g=true, r=true}) {name=Posterize Emissive Out, nodeColor=<rgba 1 0 0.1875 1>}
  NamedRerouteDeclaration(None=MaterialFunctionCall(
    Input=SceneTexture() {sceneTextureId=PPI_SceneDepth}.Color
    Input Low=ScalarParameter() {defaultValue=100000, group=1. Cell Shading, parameterName=Falloff Near Point}
    Input High=ScalarParameter() {defaultValue=150000, group=1. Cell Shading, parameterName=Falloff Far Point}
   ) {materialFunction=&RemapValueRangeNormalized}.Result) {name=Skybox Mask, nodeColor=<rgba 1 0.9141 0 1>}
  NamedRerouteDeclaration(None=@n15) {name=Cel Shader Emissive Out, nodeColor=<rgba 1 0.6094 0 1>}
  NamedRerouteDeclaration(None=ComponentMask(None=@n9) {b=true, g=true, r=true}) {name=Comic Frame Emissive Out, nodeColor=<rgba 0.9922 1 0 1>}
  ScalarParameter() {group=0. VFX Adjustments, parameterName=MasterVFXAdjustment, sortPriority=1}
  StaticSwitchParameter(True=LinearInterpolate(B=@n20.RGB, Alpha=@n24) {constA=1}, False=@n25) {parameterName=Debug}
  StaticSwitchParameter(True=LinearInterpolate(A=@n25, Alpha=@n19.Result), False=@n26) {parameterName=Debug}
 ]
}
